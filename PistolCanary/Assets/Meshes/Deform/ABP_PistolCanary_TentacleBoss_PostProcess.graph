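# Post-process Animation Blueprint for the PistolCanary tentacle boss that applies hit-impact deformation through a Control Rig on the input pose.
# Hit location, trace start and bone arrays, not populated here, feed the rig and are cleared after evaluation, acting as one-shot queues.
# On init it caches r.MeshDeformerMaxLOD as the rig's LOD threshold.
# Each post-evaluation recomputes the rig's on/off alpha from rendering detail mode or that cached value.
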AnimBlueprint /PistolCanary/Assets/Meshes/Deform/ABP_PistolCanary_TentacleBoss_PostProcess
target_skeleton=/PistolCanary/Assets/Meshes/Tentacle_PistolCanary_Creature_Skeleton
graphs [
 AnimGraph {nodes [
   n1 = Control Rig!(
    Source=Input Pose()
    Alpha=SelectFloat(A=1, B=0, bPickA=Get Enabled() -> (Enabled:bool))
    LODThreshold:int=Get MeshDeformerMaxLOD_GlobalSetting() -> (MeshDeformerMaxLOD_GlobalSetting:int=0)
    HitLocationsWS:Array of Vectors=Get HitLocationsWS() -> (HitLocationsWS:Array of Vectors)
    HitTraceStartsWS:Array of Vectors=Get HitTraceStartsWS() -> (HitTraceStartsWS:Array of Vectors)
    HitBones:Array of Names=Get Hit Bones() -> (Hit Bones:Array of Names)
    ResponseSpeed_Normalized:double=0
    Radius_Normalized:double=0
    Stiffness_Normalized:double=0
    Force_Normalized:double=0
    MaxDuration:double=1
   ) -> (Pose) {
    alphaBoolBlend {blendOption=Linear}
    controlRigAssetReference {blueprintRigClass=/PistolCanary/Anim/Test/PistolCanary_Tentacle_ImpactDeformer/CR_PistolCanary_TentacleBoss_ImpactDeform.CR_PistolCanary_TentacleBoss_ImpactDeform_C}
   }
   Output Pose(Result=@n1)
  ]}
 EventGraph {nodes [
   Event BlueprintInitializeAnimation()
   > Set MeshDeformerMaxLOD_GlobalSetting(MeshDeformerMaxLOD_GlobalSetting:int=Truncate(A=GetNumericCVar!(CVar:String=r.MeshDeformerMaxLOD) -> (OutValue:float, ReturnValue:bool))) -> (Output_Get:int=0)
   Event BlueprintPostEvaluateAnimation() {is=title K2Node_Event}
   n1 = GetRenderingDetailMode!() -> (ReturnValue:int=0)
   n2 = Get MeshDeformerMaxLOD_GlobalSetting() -> (MeshDeformerMaxLOD_GlobalSetting=0)
   > Set Enabled(Enabled:bool=OR Boolean(A="integer >= integer"(A=@n1, B=1), B="Equal (Integer)"(A=@n2, B=1))) -> (Output_Get:bool)
   n3 = Get HitLocationsWS()
   > Branch(Condition="integer > integer"(A=Length(TargetArray:Array of Vectors=@n3)))
   > Clear(TargetArray:Array of Vectors=@n3)
   > Clear(TargetArray=Get HitTraceStartsWS())
   > Clear(TargetArray:Array of Names=Get Hit Bones())
  ]}
]
variables [
 Enabled {type=boolean, default=true}
 Hit Bones {type="array<string>", default []}
 HitLocationsWS {type="array<Vector>"}
 HitTraceStartsWS {type="array<Vector>"}
 MeshDeformerMaxLOD_GlobalSetting {type=integer}
]
